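# Sound cue for the scrap pistol's degrade sound, built from borrowed Flintlock bow degrade recordings.
# A Random node picks one of three equally weighted waves, and a Modulator shifts its pitch up within a narrow random range.
# The cue plays with boosted volume on the low-ammo/degraded ranged weapon sound class.
# Its own concurrency override prevents new instances from starting once the limit is reached.
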SoundCue /PrimalGameplay/Sounds/Weapons/Scrap/Pistol/Scrap_Pistol_Degrade_Cue
root=Modulator_0
nodes [
 {id=Modulator_0, children [Random_0], properties {pitchMax=1.3, pitchMin=1.2}}
 {id=Random_0, children [WavePlayer_0,WavePlayer_1,WavePlayer_2], properties {weights [1,1,1]}}
 {id=WavePlayer_0, properties {soundWaveAssetPtr=/Game/Sounds/Fort_Weapons/Bows/Flintlock/Bow_Flintlock_Degrade_01}}
 {id=WavePlayer_1, properties {soundWaveAssetPtr=/Game/Sounds/Fort_Weapons/Bows/Flintlock/Bow_Flintlock_Degrade_02}}
 {id=WavePlayer_2, properties {soundWaveAssetPtr=/Game/Sounds/Fort_Weapons/Bows/Flintlock/Bow_Flintlock_Degrade_03}}
]
properties {
 bOverrideConcurrency=true
 maxDistance=2097152
 soundClassObject=/Game/Sounds/FortSoundClassesAndModes/Fort_WeaponRanged_LowAmmo_Degraded_PSM
 volumeMultiplier=1.5
 concurrencyOverrides {resolutionRule=PreventNew}
}
