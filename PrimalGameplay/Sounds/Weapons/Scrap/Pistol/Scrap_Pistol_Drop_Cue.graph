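# Sound cue for the scrap pistol's drop sound, routed through the Fort_Loot_Collect sound class.
# It picks one of three drop waves at equal weight, then a modulator slightly randomizes pitch and volume on top of the cue-level multipliers.
# The shared weapon-drop attenuation asset shapes its spatial falloff, and the Weapon_Drop_StopOldest concurrency set governs overlapping plays.
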SoundCue /PrimalGameplay/Sounds/Weapons/Scrap/Pistol/Scrap_Pistol_Drop_Cue
root=Modulator_0
nodes [
 {id=Modulator_0, children [Random_0], properties {pitchMax=1.1, pitchMin=1, volumeMax=0.75, volumeMin=0.65}}
 {id=Random_0, children [WavePlayer_1,WavePlayer_2,WavePlayer_3], properties {weights [1,1,1]}}
 {id=WavePlayer_1, properties {soundWaveAssetPtr=&/Scrap_Pistol_Drop_01}}
 {id=WavePlayer_2, properties {soundWaveAssetPtr=&/Scrap_Pistol_Drop_02}}
 {id=WavePlayer_3, properties {soundWaveAssetPtr=&/Scrap_Pistol_Drop_03}}
]
properties {
 attenuationSettings=&Fort_Weapon_Drop_Att
 maxDistance=3000
 pitchMultiplier=1.2
 soundClassObject=&Fort_Loot_Collect
 volumeMultiplier=0.9
 concurrencyOverrides {maxCount=1, resolutionRule=PreventNew}
 concurrencySet [&Weapon_Drop_StopOldest]
}
