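# Sound cue for the scrap pistol's gunfire, routed by default to the third-person ranged-weapon fire sound class.
# A mixer layers close, medium and distant branches at decreasing input volume, each with its own attenuation and random pick of wave variations.
# The medium and distant branches use the distant-fire sound class with random pitch and volume, while the close layer varies only volume.
# Local and global gunfire concurrency sets are assigned to limit overlapping shots.
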SoundCue /PrimalGameplay/Sounds/Weapons/Scrap/Pistol/Scrap_Pistol_Fire_Cue
root=Mixer_0
nodes [
 {
  id=Attenuation_1
  children [Random_2]
  properties {attenuationSettings=&Fort_Weapon_Projectile_Distant_Mono_Athena_Att}
 }
 {id=Attenuation_3, children [Modulator_2], properties {attenuationSettings=&Fort_Weapon_Projectile_Close_Athena_Att}}
 {
  id=Attenuation_6
  children [Random_1]
  properties {attenuationSettings=&Fort_Weapon_Projectile_Medium_Mono_Athena_Att}
 }
 {id=Mixer_0, children [SoundClass_3,Modulator_0,Modulator_1], properties {inputVolume [1,0.8,0.64]}}
 {id=Modulator_0, children [SoundClass_6], properties {pitchMax=1.15, pitchMin=0.85, volumeMax=1.15, volumeMin=0.9}}
 {id=Modulator_1, children [SoundClass_1], properties {pitchMax=1.15, pitchMin=0.85, volumeMax=1.15, volumeMin=0.9}}
 {id=Modulator_2, children [Random_0], properties {volumeMax=0.7, volumeMin=0.6}}
 {id=Random_0, children [WavePlayer_0,WavePlayer_1,WavePlayer_2,WavePlayer_3], properties {weights [1,1,1,1]}}
 {id=Random_1, children [WavePlayer_4,WavePlayer_5,WavePlayer_6], properties {weights [1,1,1]}}
 {id=Random_2, children [WavePlayer_7,WavePlayer_8,WavePlayer_9], properties {weights [1,1,1]}}
 {
  id=SoundClass_1
  children [Attenuation_1]
  properties {soundClassOverride=&Fort_WeaponRanged_Fire_Distant}
 }
 {
  id=SoundClass_3
  children [Attenuation_3]
  properties {soundClassOverride=&Fort_WeaponRanged_Fire_3P_PSM}
 }
 {
  id=SoundClass_6
  children [Attenuation_6]
  properties {soundClassOverride=&Fort_WeaponRanged_Fire_Distant}
 }
 {id=WavePlayer_0, properties {soundWaveAssetPtr=&/Scrap_Pistol_Fire_Close_01}}
 {id=WavePlayer_1, properties {soundWaveAssetPtr=&/Scrap_Pistol_Fire_Close_02}}
 {id=WavePlayer_2, properties {soundWaveAssetPtr=&/Scrap_Pistol_Fire_Close_03}}
 {id=WavePlayer_3, properties {soundWaveAssetPtr=&/Scrap_Pistol_Fire_Close_04}}
 {id=WavePlayer_4, properties {soundWaveAssetPtr=&/Scrap_Pistol_Fire_Medium_01}}
 {id=WavePlayer_5, properties {soundWaveAssetPtr=&/Scrap_Pistol_Fire_Medium_02}}
 {id=WavePlayer_6, properties {soundWaveAssetPtr=&/Scrap_Pistol_Fire_Medium_03}}
 {id=WavePlayer_7, properties {soundWaveAssetPtr=&/Scrap_Pistol_Fire_Distant_01}}
 {id=WavePlayer_8, properties {soundWaveAssetPtr=&/Scrap_Pistol_Fire_Distant_02}}
 {id=WavePlayer_9, properties {soundWaveAssetPtr=&/Scrap_Pistol_Fire_Distant_03}}
]
properties {
 maxDistance=27500
 pitchMultiplier=0.9
 soundClassObject=&Fort_WeaponRanged_Fire_3P_PSM
 volumeMultiplier=0.8
 attenuationOverrides {distanceAlgorithm=NaturalSound, falloffDistance=5000, attenuationShapeExtents {x=1000}}
 concurrencyOverrides {bLimitToOwner=true, maxCount=3, resolutionRule=StopOldest}
 concurrencySet [&Weapon_Gunfire_Local_Conc, &Weapon_Gunfire_Global_Conc]
}
