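# Sound cue for the low-ammo sound of the Scrap Pistol weapon.
# A single wave player is the root, playing a reused Flintlock Revolver low-ammo sample with no randomization or mixing.
# The cue routes to the degraded ranged-weapon low-ammo sound class and overrides concurrency so new instances are refused once the limit is reached.
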SoundCue /PrimalGameplay/Sounds/Weapons/Scrap/Pistol/Scrap_Pistol_LowAmmo_Cue
root=WavePlayer_0
nodes [{id=WavePlayer_0, properties {soundWaveAssetPtr=/Game/Sounds/Fort_Weapons/FlintLock/Flintlock_Revolver_LowAmmo_02}}]
properties {
 bOverrideConcurrency=true
 maxDistance=2097152
 soundClassObject=/Game/Sounds/FortSoundClassesAndModes/Fort_WeaponRanged_LowAmmo_Degraded_PSM
 concurrencyOverrides {resolutionRule=PreventNew}
}
