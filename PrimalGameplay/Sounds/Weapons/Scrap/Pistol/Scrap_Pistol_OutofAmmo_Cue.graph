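# Sound cue for the scrap pistol's out-of-ammo sound, built from reused launcher and flintlock revolver waves.
# A mixer layers one of four randomly chosen launcher out-of-ammo clips, with slight volume variation, over a fixed flintlock low-ammo clip.
# An outer modulator varies the combined result, which the cue pitches up and lowers in volume.
# It uses the third-person ranged fire sound class and caps concurrency at two, stopping the oldest.
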SoundCue /PrimalGameplay/Sounds/Weapons/Scrap/Pistol/Scrap_Pistol_OutofAmmo_Cue
root=Modulator_2
nodes [
 {id=Mixer_0, children [Modulator_1,WavePlayer_8], properties {inputVolume [1,1]}}
 {id=Modulator_1, children [Random_1], properties {pitchMax=1, pitchMin=1, volumeMax=0.9, volumeMin=0.8}}
 {id=Modulator_2, children [Mixer_0], properties {volumeMax=0.7, volumeMin=0.6}}
 {id=Random_1, children [WavePlayer_4,WavePlayer_5,WavePlayer_6,WavePlayer_7], properties {weights [1,1,1,1]}}
 {id=WavePlayer_4, properties {soundWaveAssetPtr=/Game/Sounds/Fort_Weapons/OutofAmmo/Launcher_OutofAmmo_01}}
 {id=WavePlayer_5, properties {soundWaveAssetPtr=/Game/Sounds/Fort_Weapons/OutofAmmo/Launcher_OutofAmmo_02}}
 {id=WavePlayer_6, properties {soundWaveAssetPtr=/Game/Sounds/Fort_Weapons/OutofAmmo/Launcher_OutofAmmo_03}}
 {id=WavePlayer_7, properties {soundWaveAssetPtr=/Game/Sounds/Fort_Weapons/OutofAmmo/Launcher_OutofAmmo_04}}
 {id=WavePlayer_8, properties {soundWaveAssetPtr=/Game/Sounds/Fort_Weapons/FlintLock/Flintlock_Revolver_LowAmmo_02}}
]
properties {
 bOverrideConcurrency=true
 maxDistance=2097152
 pitchMultiplier=1.2
 soundClassObject=&Fort_WeaponRanged_Fire_3P_PSM
 volumeMultiplier=0.65
 attenuationOverrides {falloffDistance=1600}
 concurrencyOverrides {maxCount=2, resolutionRule=StopOldest}
}
