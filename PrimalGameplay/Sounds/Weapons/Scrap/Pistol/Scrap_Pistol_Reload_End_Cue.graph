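# Sound cue for the end of the Scrap pistol's reload.
# A single wave player feeds the close reload sample straight to the output, with no randomization.
# It plays as quiet weapon foley with reload attenuation and a concurrency override that prevents new instances once the limit is reached.
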SoundCue /PrimalGameplay/Sounds/Weapons/Scrap/Pistol/Scrap_Pistol_Reload_End_Cue
root=WavePlayer_0
nodes [{id=WavePlayer_0, properties {soundWaveAssetPtr=&/Scrap_Pistol_Reload_Close_01}}]
properties {
 attenuationSettings=&Fort_Weapon_Reload_Att
 bOverrideConcurrency=true
 maxDistance=2250
 soundClassObject=&Fort_WeaponFoley
 volumeMultiplier=0.301
 concurrencyOverrides {resolutionRule=PreventNew}
}
