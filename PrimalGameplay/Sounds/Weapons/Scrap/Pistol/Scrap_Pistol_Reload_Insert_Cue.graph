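# Sound cue for the insert step of the Scrap Pistol reload in PrimalGameplay.
# A single wave player feeds the output directly, reusing the SixShooter reload insert sample with no randomization or modulation.
# It plays at reduced volume through the weapon foley sound class with weapon reload attenuation.
# A concurrency override prevents new instances from starting once the limit is reached.
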SoundCue /PrimalGameplay/Sounds/Weapons/Scrap/Pistol/Scrap_Pistol_Reload_Insert_Cue
root=WavePlayer_1
nodes [{id=WavePlayer_1, properties {soundWaveAssetPtr=/Game/Sounds/Fort_Weapons/pistol/SixShooter/SixShooter_ReloadInsert_01}}]
properties {
 attenuationSettings=&Fort_Weapon_Reload_Att
 bOverrideConcurrency=true
 maxDistance=2250
 soundClassObject=&Fort_WeaponFoley
 volumeMultiplier=0.752
 concurrencyOverrides {resolutionRule=PreventNew}
}
